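; Searchable beacon actor for the Mash LTM that shows offline, charging or malfunction states as match phases change.
; It retries finding the Mash mutator after delays, then binds phase changes to material state, an aftermath pulse and sound.
; During Survival a one-second timer drives material charge against a curve-table maximum.
; Chest-derived logic handles audio and highlight cleanup on search or loot, plus mobile wiggle and outline through dynamic materials.
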(event EventBeginPlay
  ;[1 1 1] Begin Play
  (Utilities|FlowControl|Sequence
    (:then_0
      (Networking|CanExecuteCosmeticEvents
        (:True
          (CallFunction|CreateMobileMIDs)
          (Variables|Default|SetChimeTimer))))
    (:then_1
      ;[1 1 1] Get Mutator
      (Utilities|FlowControl|Delay 1e+01 :!label _label)
      (bind _asfort_game_state_athena (Utilities|Casting|CastToFortGameStateAthena (Game|GetGameState)))
      (bind _asfort_athena_mutator_mash (Utilities|Casting|CastToFortAthenaMutator_Mash (Modifier|GetMutatorbyClass (Variables|/Script/FortniteGame.FortGameStateAthena|GetMutatorListComponent _asfort_game_state_athena) "/Game/Athena/Playlists/Mash/MashMutator.MashMutator_C"))
        (:then
          (EventDispatchers|BindEventtoOnPhaseChanged _asfort_athena_mutator_mash (delegate Custom|MashPhaseChanged) :!class /Script/FortniteGame.FortAthenaMutator_Mash))
        (:CastFailed
          (goto _label))))))

(event Custom|MashPhaseChanged
  (params (NewPhase :type Integer))
  (bind _p_mash_pulse (Variables|Default|GetP_Mash_Pulse))
  (bind _sm_mash_beacon_interact_temp (Variables|Default|Getsm_Mash_Beacon_Interact_TEMP))
  ;[1 1 1] Set the Beacon States : 0 = Offline, 1 Charging, 2 Malfunction.
  (switch Utilities|FlowControl|Switch|SwitchonEMashPhase (Variables|/Script/FortniteGame.FortAthenaMutator_Mash|GetCurrentPhase (stale EventBeginPlay _asfort_athena_mutator_mash))
    (:NotStarted
      _)
    (:Preparation
      _)
    (:Survival
      (bind _output_get (Variables|Default|SetBeaconState 1))
      (Rendering|Material|SetScalarParameterValueonMaterials _sm_mash_beacon_interact_temp "State" (Math|Conversions|ToFloat(Integer) _output_get))
      (bind _returnvalue (Utilities|Time|SetTimerbyEvent (delegate Custom|UpdateChargingPercent) 1.0 true false 0.1)))
    (:AfterMath
      (Components|Activation|SetActive _p_mash_pulse true false)
      (Audio|PlaySoundatLocation "/Game/Athena/Sounds/LTMs/MASH/MASH_Beacon_Activate_Cue.MASH_Beacon_Activate_Cue" (Transformation|GetWorldLocation _p_mash_pulse)))
    (:Moving
      _)
    (:FinalPhase
      (bind _output_get_2 (Variables|Default|SetBeaconState 2))
      (Rendering|Material|SetScalarParameterValueonMaterials _sm_mash_beacon_interact_temp "State" (Math|Conversions|ToFloat(Integer) _output_get_2))))
  (bind _output_get_3 (Variables|Default|SetBeaconState 0))
  (Rendering|Material|SetScalarParameterValueonMaterials _sm_mash_beacon_interact_temp "State" (Math|Conversions|ToFloat(Integer) _output_get_3))
  ;[1 1 1] Reset Charge % Values
  (Utilities|Time|PauseTimerbyHandle _returnvalue)
  (Variables|Default|SetCurrentChargeTime 0.0)
  (Variables|Default|SetChargingPercent 0.0))

(event Custom|Mobile_OnInteract
  ;[1 1 1] Mobile Wiggle on Interact
  (Networking|CanExecuteCosmeticEvents
    (:True
      (bind (_direction _lerp) (Timeline|MobileOnInteractTL Play)
        (:Update
          (for _array_element (Variables|Default|GetMIDs)
            (Rendering|Material|SetScalarParameterValue _array_element "WiggleOnInteractAmount" (Math|Float|Lerp 0.0 (Variables|Default|GetMobileWiggleAmount) _lerp) :!class /Script/Engine.MaterialInstanceDynamic)))))))

(event Custom|Mobile_Outline_On
  ;[1 1 1] Mobile Outline
  (for _array_element (Variables|Default|GetMIDs)
    (Rendering|Material|SetScalarParameterValue _array_element "Outline_Control" 1.0 :!class /Script/Engine.MaterialInstanceDynamic)
    (Rendering|Material|SetScalarParameterValue _array_element "Pulse_Control" 1.0 :!class /Script/Engine.MaterialInstanceDynamic)
    (Rendering|Material|SetVectorParameterValue _array_element "Color" "(R=1.000000,G=1.000000,B=1.000000,A=1.000000)" :!class /Script/Engine.MaterialInstanceDynamic)))

(event Custom|Mobile_OnDisconnect
  ;[1 1 1] Mobile Wiggle on Interact
  (Networking|CanExecuteCosmeticEvents
    (:True
      (Timeline|MobileOnInteractTL Reverse))))

(event Custom|UpdateChargingPercent
  ;[1 1 1] Set the % of the Charge of the Beacon
  (bind _result (Math|Float|IncrementFloat (Variables|Default|GetCurrentChargeTime))
    (:%20%20
      (bind _output_get (Variables|Default|SetCurrentChargeTime _result))
      (bind _output_get_2 (Variables|Default|SetChargingPercent (/ _output_get (Variables|Default|GetMaxChargeTime))))
      (Rendering|Material|SetScalarParameterValueonMaterials (Variables|Default|Getsm_Mash_Beacon_Interact_TEMP) "Charge" _output_get_2))))

(event Custom|Mobile_Outline_Off
  ;[1 1 1] Mobile Outline
  (for _array_element (Variables|Default|GetMIDs)
    (Rendering|Material|SetScalarParameterValue _array_element "Outline_Control" 0.0 :!class /Script/Engine.MaterialInstanceDynamic)
    (Rendering|Material|SetScalarParameterValue _array_element "Pulse_Control" 0.0 :!class /Script/Engine.MaterialInstanceDynamic)))

(event EventOnSetSearched
  (Audio|Components|Audio|FadeOut (Variables|Default|GetChest_Ambient_Sound) 0.5 0.0 "Linear" :!class /Script/Engine.AudioComponent)
  (Utilities|Time|ClearandInvalidateTimerbyHandle (Variables|Default|GetChimeTimer))
  (Variables|Building|SetbDestroyOnPlayerBuildingPlacement true) ; Allows players to build over the opened chest mesh (but not when there is loot in there)
  (CallFunction|CleanupWiggleMIDs)
  (Rendering|SetRenderCustomDepth (Variables|Debug|GetStaticMeshComponent) false))

(event EventDestroyed
  (Audio|Components|Audio|FadeOut (Variables|Default|GetChest_Ambient_Sound) 0.5 0.0 "Linear" :!class /Script/Engine.AudioComponent)
  (Utilities|Time|ClearandInvalidateTimerbyHandle (Variables|Default|GetChimeTimer))
  (Variables|Building|SetbDestroyOnPlayerBuildingPlacement true) ; Allows players to build over the opened chest mesh (but not when there is loot in there)
  (CallFunction|CleanupWiggleMIDs)
  (Rendering|SetRenderCustomDepth (Variables|Debug|GetStaticMeshComponent) false))

(event EventOnLoot
  (Audio|Components|Audio|FadeOut (Variables|Default|GetChest_Ambient_Sound) 0.5 0.0 "Linear" :!class /Script/Engine.AudioComponent)
  (Utilities|Time|ClearandInvalidateTimerbyHandle (Variables|Default|GetChimeTimer))
  (Variables|Building|SetbDestroyOnPlayerBuildingPlacement true) ; Allows players to build over the opened chest mesh (but not when there is loot in there)
  (CallFunction|CleanupWiggleMIDs)
  (Rendering|SetRenderCustomDepth (Variables|Debug|GetStaticMeshComponent) false))

(event EventOnDeathPlayEffects
  (params (Damage :type Float%20%28single-precision%29) (DamageTags :type Gameplay%20Tag%20Container%20Structure%20%28by%20ref%29) (Momentum :type Vector :default "0, 0, 0") (HitInfo :type Hit%20Result%20Structure%20%28by%20ref%29) (InstigatedBy :type Fort%20Pawn%20Object%20Reference) (DamageCauser :type Actor%20Object%20Reference) (EffectContext :type Gameplay%20Effect%20Context%20Handle%20Structure))
  (Audio|Components|Audio|FadeOut (Variables|Default|GetChest_Ambient_Sound) 0.5 0.0 "Linear" :!class /Script/Engine.AudioComponent))

(fn BlueprintCanInteract
  (params (InteractingPawn :type Fort%20Pawn%20Object%20Reference) (InteractionBeingAttempted :type EInteractionBeingAttempted%20Enum :default "FirstInteraction") (InteractionType :type TInteractionType%20Enum :default "IT_NoInteraction"))
  (return (Variables|Default|GetCanInteract)))

(fn CleanupWiggleMIDs)

(fn CreateMobileMIDs)

(fn GetMaxAudibleDistance
  (bind _chest_ambient_sound (Variables|Default|GetChest_Ambient_Sound))
  (bind _sound (Variables|/Script/Engine.AudioComponent|GetSound _chest_ambient_sound))
  (Utilities|IsValid _sound
    (:Is%20Valid
      (return (Variables|/Script/Engine.SoundBase|GetMaxDistance _sound)))
    (:Is%20Not%20Valid
      (return 0.0))))

(fn OnSetCustomDepthStencilValue
  (params (PrimComponents :type Array%20of%20Primitive%20Component%20Object%20References) (bUseCustomDepth :type Boolean) (StencilValue :type Integer))
  (for _array_element PrimComponents
    (Rendering|SetCustomDepthStencilValue _array_element (Variables|GetStencilValue))
    (if (Variables|GetbUseCustomDepth)
      (Rendering|SetRenderCustomDepth _array_element true)))
  (return true))

(fn ShouldDie
  (params (Damage :type Float%20%28single-precision%29) (EventInstigator :type Controller%20Object%20Reference) (DamageCauser :type Actor%20Object%20Reference) (bIsTest :type Boolean) (EffectContext :type Gameplay%20Effect%20Context%20Handle%20Structure%20%28by%20ref%29))
  (bind _asfort_player_controller_outpost (Utilities|Casting|CastToFortPlayerControllerOutpost EventInstigator)
    (:then
      (return (Outpost|IsOutpostOwner _asfort_player_controller_outpost)))
    (:CastFailed
      (return true))))

(fn ConstructionScript
  (bind _outxy (Fortnite|EvaluateCurveTableRow (Variables|Default|GetChargeTimeCurveTableHandle) 3.0))
  (Variables|Default|SetMaxChargeTime _outxy)
  (|CallParentFunction))
