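; MashChestVFX is a Blueprint that controls the in-world chest particle effect for the Mash playlist.
; Its construction script sets the "Color" parameter of the P_Mash_Chest_InWorld particle component from the FXColor variable, so the tint is configurable per instance or subclass.
; A custom DisableVFX event deactivates that particle component.
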(event Custom|DisableVFX
  (Components|Activation|Deactivate (Variables|Default|GetP_Mash_Chest_InWorld)))

(fn ConstructionScript
  (Effects|Components|ParticleSystem|SetColorParameter (Variables|Default|GetP_Mash_Chest_InWorld) "Color" (Variables|Default|GetFXColor)))
