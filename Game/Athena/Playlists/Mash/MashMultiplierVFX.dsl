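; Visual effects actor for the Mash playlist's score multiplier pickup, with an in-world effect tinted by FXColor in the construction script.
; Activate Explosion VFX takes the interacting pawn and, if valid, plays a local or non-local pulse depending on whether that pawn is locally controlled.
; It also plays the pickup burst, stops the in-world effect and destroys the actor after a short delay.
; DisableVFX deactivates all four particle components.
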(event Custom|Activate%20Explosion%20VFX
  (params (Interacting%20Pawn :type Fort%20Pawn%20Object%20Reference))
  (Utilities|IsValid Interacting%20Pawn
    (:Is%20Valid
      (if (Pawn|IsLocallyControlled Interacting%20Pawn)
        (Components|Activation|Activate (Variables|Default|GetP_Mash_Pulse_ScoreObject_Local) false)
        (else
          (Components|Activation|Activate (Variables|Default|GetP_Mash_Pulse_ScoreObject) false)))
      (Components|Activation|Activate (Variables|Default|GetP_Mash_Scoring_PickUp) false)
      (Components|Activation|Deactivate (Variables|Default|GetP_Mash_Scoring_InWorld))
      (Utilities|FlowControl|Delay 1.0)
      (Actor|DestroyActor))))

(event Custom|DisableVFX
  (Components|Activation|Deactivate (Variables|Default|GetP_Mash_Pulse_ScoreObject_Local))
  (Components|Activation|Deactivate (Variables|Default|GetP_Mash_Scoring_PickUp))
  (Components|Activation|Deactivate (Variables|Default|GetP_Mash_Scoring_InWorld))
  (Components|Activation|Deactivate (Variables|Default|GetP_Mash_Pulse_ScoreObject)))

(fn ConstructionScript
  (Effects|Components|ParticleSystem|SetColorParameter (Variables|Default|GetP_Mash_Scoring_InWorld) "Color" (Variables|Default|GetFXColor)))
