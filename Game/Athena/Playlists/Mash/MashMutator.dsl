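; Fortnite Athena mutator Blueprint for the Mash mode, handling a no-death marker and end-of-match team score.
; On authority, BeginPlay binds handlers to phase change and victory events.
; The first Preparation phase applies a GameplayEffect to all player pawns, granting a tag for checking deaths.
; Victory prints the first player's team score, then, if the quest check passes, loops players validating controllers without further action, despite a score-reporting comment.
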(event Custom|OnPhaseChanged_Event
  (params (NewPhase :type Integer))
  ;[1 1 1] Apply GE to all player pawns to grant a tag that we check against to see if they died during the match
  (switch Utilities|FlowControl|Switch|SwitchonEMashPhase (Variables|FortAthenaMutator_Mash|GetCurrentPhase)
    (:Preparation
      (if (not (Variables|Default|GetBuffApplied))
        (bind _returnvalue (Fortnite|AuthorityGetFortPossessedPlayerPawns))
        (for _array_element _returnvalue
          (GameplayEffects|ApplyGameplayEffectToTarget :self (Ability|GetAbilitySystemComponent _array_element) :GameplayEffectClass (Variables|ReadOnly|GetHasNotDiedGE) :Target (Ability|GetAbilitySystemComponent _array_element) :Level 0.0))
        (Variables|Default|SetBuffApplied true)))))

(event EventBeginPlay
  (Networking|SwitchHasAuthority
    (:Authority
      (EventDispatchers|BindEventtoOnPhaseChanged (delegate Custom|OnPhaseChanged_Event))
      (EventDispatchers|BindEventtoOnMashVictory (delegate Custom|MashVictory)))))

(event Custom|MashVictory
  (bind _playerdataarray (Variables|FortAthenaMutator_Mash|GetPlayerDataArray))
  (bind _entries (Utilities|Struct|BreakMashPlayerDataArray _playerdataarray))
  (bind _returnvalue 0)
  (bind _output (Utilities|Array|Get(aref) _entries _returnvalue))
  (bind _playerstate (Utilities|Struct|BreakMashPlayerDataEntry _output))
  (bind _returnvalue_2 (Game|GetGameState))
  ;[0 0.1511 1] Report score for all players at the end of the match
  (Utilities|FlowControl|Sequence
    (:then_0
      ;;[0.2663 0.3722 1] Get Total Team Score
      (Variables|Default|SetTotalTeamScore 0)
      (if (Utilities|Array|IsValidIndex _entries _returnvalue)
        (Utilities|IsValid _playerstate
          (:Is%20Valid
            (Variables|Default|SetTotalTeamScore (Score|GetTeamScore (Class|FortPlayerStateAthena|GetTeam _playerstate)))))))
    (:then_1
      (Development|PrintString (Utilities|String|Append "MASH: Total Score: " (Utilities|String|ToString(Integer) (Variables|Default|GetTotalTeamScore))))
      (Utilities|IsValid _returnvalue_2
        (:Is%20Valid
          (bind _cancompletequest (Class|AthenaQuestFunctionLibrary|AthenaCheckCreativeMode _returnvalue_2))
          (if _cancompletequest
            (for _array_element (Utilities|Struct|BreakMashPlayerDataArray (Variables|FortAthenaMutator_Mash|GetPlayerDataArray))
              (Utilities|IsValid (Utilities|Struct|BreakMashPlayerDataEntry _array_element)
                (:Is%20Valid
                  (Utilities|IsValid (Fortnite|GetFortPlayerControllerfromActor (Utilities|Struct|BreakMashPlayerDataEntry _array_element))))))))))))

(fn ConstructionScript)
