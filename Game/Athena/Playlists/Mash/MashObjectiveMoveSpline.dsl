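; Spline actor for the Mash playlist that, going by its name, defines a path for moving objectives.
; GetSplineComponent exposes the spline component to other objects.
; The construction script prints the actor's display name and spline length when running in editor or PIE.
; BeginPlay, overlap and Tick events are empty, so no movement logic lives here.
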(event EventBeginPlay)

(event Collision|EventActorBeginOverlap
  (params (OtherActor :type Actor%20Object%20Reference)))

(event EventTick
  (params (DeltaSeconds :type Float%20%28single-precision%29)))

(fn GetSplineComponent
  (return (Variables|Default|GetSpline)))

(fn ConstructionScript
  (if (Editor|InEditororPIE)
    (Development|PrintString (Utilities|String|Append (Utilities|String|Append (Utilities|GetDisplayName self) " Spline Length: ") (Utilities|String|ToString(Float) (Spline|GetSplineLength (Variables|Default|GetSpline)))))))
