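; Poison cloud area hazard for the Mash playlist that damages pawns inside it over time.
; Curve tables set the damage sphere's radius at construction and the cloud's duration at BeginPlay.
; On the server, overlap begin applies a poison gameplay effect to pawns matching a tag query, and overlap end removes it by tag.
; Timers step down Niagara visibility near expiry, then disable collision, stop emission and fade audio before lifespan destroys the actor.
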(event EventBeginPlay
  ;[1 1 1] Set up cloud.
  (Collision|SetCollisionComponent (Variables|Default|GetDamageArea))
  (bind _outxy (Fortnite|EvaluateCurveTableRow (Variables|Default|GetCloudDurationCurveTable)))
  (Utilities|Time|SetTimerbyFunctionName 0 "EndAOE" _outxy)
  (Actor|SetLifeSpan (+ _outxy (Variables|Default|GetAOEFXFadeTime)))
  (Utilities|Time|SetTimerbyEvent (delegate Custom|FadeVFX) 0.05 true false (- _outxy 1.0)))

(event Custom|FadeVFX
  (bind _ns_poisoncloud (Variables|Default|GetNS_PoisonCloud))
  (bind _aoefadevfx (Variables|Default|GetAOEFadeVFX))
  (if (not (<= _aoefadevfx 0.0))
    (Effects|Components|ParticleSystem|SetFloatParameter _ns_poisoncloud "Visibility" _aoefadevfx)
    (Variables|Default|SetAOEFadeVFX (- _aoefadevfx 0.05))
    (else
      (Effects|Components|ParticleSystem|SetFloatParameter _ns_poisoncloud "Visibility" 0.0))))

(event Custom|EndAOE
  ;[1 1 1] Stop AOE by disabling collision. Also stop AOE emitter so it can fade out.
  (Collision|SetCollisionEnabled (Variables|Default|GetDamageArea) "NoCollision")
  (Components|Activation|SetActive (Variables|Default|GetNS_PoisonCloud) false false)
  (Audio|Components|Audio|FadeOut (Variables|Default|GetLoopingAudio) 1 0.0 "Linear" :!class /Script/Engine.AudioComponent))

(event OnComponentBeginOverlap(DamageArea)
  (params (OverlappedComponent :type Primitive%20Component%20Object%20Reference) (OtherActor :type Actor%20Object%20Reference) (OtherComp :type Primitive%20Component%20Object%20Reference) (OtherBodyIndex :type Integer) (bFromSweep :type Boolean) (SweepResult :type Hit%20Result%20Structure%20%28by%20ref%29))
  ;[1 1 1] Apply Poison Damage AoE
  (Networking|SwitchHasAuthority
    (:Authority
      (bind _asfort_pawn (Utilities|Casting|CastToFortPawn OtherActor))
      (if (GameplayTags|DoesContainerMatchTagQuery (Variables|/Script/FortniteGame.FortPawn|GetGameplayTags _asfort_pawn) (Variables|Default|GetTQ_PoisonDamageRules))
        (GameplayEffects|ApplyGameplayEffectToTarget :self (Ability|GetAbilitySystemComponent _asfort_pawn) :GameplayEffectClass "/Game/Athena/Playlists/Mash/GameEffects/GE_Mash_PoisonDot.GE_Mash_PoisonDot_C" :Target (Ability|GetAbilitySystemComponent _asfort_pawn) :Level 0.0)))))

(event OnComponentEndOverlap(DamageArea)
  (params (OverlappedComponent :type Primitive%20Component%20Object%20Reference) (OtherActor :type Actor%20Object%20Reference) (OtherComp :type Primitive%20Component%20Object%20Reference) (OtherBodyIndex :type Integer))
  ;[1 1 1] Remove Poison Damage AoE
  (Networking|SwitchHasAuthority
    (:Authority
      (bind _asfort_pawn (Utilities|Casting|CastToFortPawn OtherActor))
      (if (GameplayTags|DoesContainerMatchTagQuery (Variables|/Script/FortniteGame.FortPawn|GetGameplayTags _asfort_pawn) (Variables|Default|GetTQ_PoisonDamageRules))
        (GameplayEffects|RemoveActiveEffectswithTags (Ability|GetAbilitySystemComponent _asfort_pawn) "(GameplayTags=((TagName=\"Gameplay.Damage.Physical.Poison\")))")))))

(fn ConstructionScript
  (bind _outxy (Fortnite|EvaluateCurveTableRow (Variables|Default|GetCloudRadiusCurveTable)))
  (Components|Sphere|SetSphereRadius (Variables|Default|GetDamageArea) _outxy true))
