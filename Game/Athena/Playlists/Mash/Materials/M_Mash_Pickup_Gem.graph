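# Default Lit material for the gem pickup in the Mash playlist, with a two-tone purple surface, magenta internal glow and sine-driven vertical bobbing.
# An inverted Fresnel on a flattened turbulence normal blends the two colors.
# It also masks the glow, together with a cloud-distorted distance falloff from object center and two scrolling screen-space noises that fake interior depth.
# Tiling noise times Fresnel blends between roughness parameters.
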Material /Game/Athena/Playlists/Mash/Materials/M_Mash_Pickup_Gem
graph {
 properties {pixelDepthOffsetMode=PDOM_Legacy, previewMesh=/Game/Athena/Playlists/Mash/Meshes/sm_mash_pickup_new, usedShadingModels=Default Lit}
 nodes [
  n1 = TextureCoordinate()
  n2 = TextureSample(UVs=@n1) {samplerType=SAMPLERTYPE_Normal, texture=/Game/Effects/Fort_Effects/Textures/Plasma/T_Turbulence_Seamless_N}
  n3 = Constant() {r=0.95}
  n4 = Fresnel(
   ExponentIn=Constant() {r=0.125}
   Normal=MaterialFunctionCall(
    BaseNormal=VertexNormalWS()
    AdditionalNormal=MaterialFunctionCall(Normal=@n2.RGB, Flatness=@n3) {materialFunction=&FlattenNormal}.Result
   ) {materialFunction=&BlendAngleCorrectedNormals}.Result
  ) {baseReflectFraction=0}
  n5 = OneMinus(None=@n4)
  n6 = Saturate(None=Multiply(A=Power(Base=@n5) {constExponent=1.255}) {constB=10})
  LinearInterpolate(
   A=VectorParameter() {parameterName=Color2, defaultValue=<rgba 0.0173 0 0.125 1>}.RGB
   B=VectorParameter() {parameterName=Color1, defaultValue=<rgba 0.095 0.0198 0.4219 1>}.RGB
   Alpha=@n6
  )
  n7 = ScalarParameter() {parameterName=Distort U Mult}
  n8 = ScalarParameter() {parameterName=Distort V Mult}
  n9 = TextureSample(UVs=Multiply(A=TextureCoordinate(), B=AppendVector(A=@n7, B=@n8))) {samplerType=SAMPLERTYPE_LinearColor, texture=/Game/Global/Textures/Default/Tileable_Patterns/Cloud_Diffuse}
  n10 = ScalarParameter() {defaultValue=0.23, parameterName=Edge Distort Multiplier}
  n11 = Multiply(A=Subtract(A=ObjectPositionWS(), B=WorldPosition().XYZ), B=VectorParameter() {parameterName=Mult, defaultValue=<rgba 0.888 0.888 0.4 1>}.RGB)
  n12 = VectorParameter() {parameterName=Offset, defaultValue=<rgba 0 0 2.5 1>}
  n13 = MaterialFunctionCall(Vector 3=Add(A=Multiply(A=@n9.G, B=@n10) {constB=0.15}, B=Add(A=@n11, B=@n12.RGB))) {materialFunction=/Engine/Functions/Engine_MaterialFunctions02/Utility/VectorLength}
  n14 = Multiply(A=ObjectRadius(), B=ScalarParameter() {defaultValue=0.35, parameterName=Object Radius Multiplier})
  n15 = Power(Base=Saturate(None=Subtract(A=@n13.V3 Length, B=@n14)), Exp=ScalarParameter() {defaultValue=1, parameterName=FinalPower})
  n16 = Clamp(None=ComponentMask(None=OneMinus(None=@n15)) {b=true})
  n17 = ScreenPosition()
  n18 = Time()
  n19 = Constant2Vector() {g=0.06}
  n20 = TextureSample(UVs=Add(A=@n17.ViewportUV, B=Multiply(A=@n18, B=@n19))) {samplerType=SAMPLERTYPE_LinearColor, texture=/Game/Effects/Fort_Effects/Effects/Environment/Desert/Dust/Organize/T_Turbulence_Seamless}
  n21 = Power(Base=@n20.G) {constExponent=3}
  n22 = Multiply(A=@n18, B=Constant2Vector() {g=0.02, r=0.03})
  n23 = Power(Base=TextureSample(UVs=Add(A=@n17.ViewportUV, B=@n22)) {texture=/Game/Global/Textures/Default/Noise/T_Tiling_Noise_01}.G) {constExponent=0.5}
  LinearInterpolate(
   B=VectorParameter() {parameterName=Internal Glow, defaultValue=<rgba 1.1118 0 1.2 1>}.RGB
   Alpha=Multiply(A=Multiply(A=@n6, B=@n16), B=Multiply(A=@n21, B=@n23))
  )
  n24 = TextureSample(UVs=@n1) {texture=&T_Tiling_Noise_01}
  LinearInterpolate(
   A=ScalarParameter() {parameterName=Roughness_low}
   B=ScalarParameter() {parameterName=Roughness_High}
   Alpha=Multiply(A=ComponentMask(None=@n24.G) {r=true}, B=@n4)
  ) {constA=1, constB=0}
  n25 = Constant()
  n26 = Multiply(A=Time(), B=ScalarParameter() {defaultValue=0.25, parameterName=Bobbing Speed})
  Multiply(
   A=AppendVector(A=ComponentMask(None=@n25) {g=true, r=true}, B=Sine(None=@n26))
   B=ScalarParameter() {defaultValue=4, parameterName=Bob Amount}
  ) {constB=2}
  ScalarParameter() {parameterName=Metallic}
  VectorParameter() {parameterName=SpecColor, defaultValue=<rgba 4 0 2.8968 1>}
 ]
}
