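; Searchable score multiplier pickup for the Mash LTM that sends the interacting player's controller in a gameplay event to its granted ability set.
; Looting triggers explosion VFX and sound, hides the mesh, then destroys the actor.
; Projectile movement stops only when local and replicated rest flags agree, resuming when the landed-on building dies.
; A cosmetic timer toggles the info plate by player distance, while mobile MIDs drive outline and wiggle effects.
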(event EventEndPlay
  (params (EndPlayReason :type EEndPlayReason%20Enum :default "Destroyed"))
  ;[1 0.5783 0] EndPlay
  (|CallParentFunction EndPlayReason)
  (bind _landedbuilding (Variables|Default|GetLandedBuilding)
    (:then
      (EventDispatchers|UnbindEventfromOnDied _landedbuilding (EventDispatchers|CreateEvent :!event Custom|OnLandedBuildingDestroyed))
      (Variables|Default|SetLandedBuilding 0))
    (:else)))

(event Custom|Mobile_Outline_On
  ;[1 1 1] Mobile Outline
  (for _array_element (Variables|Default|GetMIDs)
    (Rendering|Material|SetScalarParameterValue _array_element "Outline_Control" 1.0 :!class /Script/Engine.MaterialInstanceDynamic)
    (Rendering|Material|SetScalarParameterValue _array_element "Pulse_Control" 1.0 :!class /Script/Engine.MaterialInstanceDynamic)
    (Rendering|Material|SetVectorParameterValue _array_element "Color" "(R=1.000000,G=1.000000,B=1.000000,A=1.000000)" :!class /Script/Engine.MaterialInstanceDynamic)))

(event Custom|Mobile_Outline_Off
  ;[1 1 1] Mobile Outline
  (for _array_element (Variables|Default|GetMIDs)
    (Rendering|Material|SetScalarParameterValue _array_element "Outline_Control" 0.0 :!class /Script/Engine.MaterialInstanceDynamic)
    (Rendering|Material|SetScalarParameterValue _array_element "Pulse_Control" 0.0 :!class /Script/Engine.MaterialInstanceDynamic)))

(event EventBeginPlay
  ;[0.01829 1 0] Begin Play
  (Interaction|BindEventtoOnInteract (delegate Custom|OnInteracted))
  (Ability|EquipFortAbilitySet self "/Game/Athena/Playlists/Mash/GameAbilities/GAS_ScoreMultiplier.GAS_ScoreMultiplier" (Variables|Default|GetAbilitySystem))
  (CallFunction|UpdateProjectileMovement)
  (Networking|CanExecuteCosmeticEvents
    (:True
      (bind _returnvalue (Utilities|Time|SetTimerbyEvent (EventDispatchers|CreateEvent :!event Custom|CheckShouldDisplayUI) 0.4 true))
      (Variables|Default|SetUICheckTimer _returnvalue)
      (CallFunction|CreateMobileMIDs)
      (if (Variables|Container|GetbAlreadySearched)
        (CallFunction|SetLightVisibility false)
        (else
          (CallFunction|SetLightVisibility true))))))

(event Custom|Mobile_OnInteract
  ;[1 1 1] Mobile Wiggle on Interact
  (Networking|CanExecuteCosmeticEvents
    (:True
      (bind (_direction _lerp) (Timeline|MobileOnInteractTL Play)
        (:Update
          (for _array_element (Variables|Default|GetMIDs)
            (Rendering|Material|SetScalarParameterValue _array_element "WiggleOnInteractAmount" (Math|Float|Lerp 0.0 (Variables|Default|GetMobileWiggleAmount) _lerp) :!class /Script/Engine.MaterialInstanceDynamic)))))))

(event Custom|OnInteracted
  (params (SelfActor :type Building%20Actor%20Object%20Reference) (InteractingPawn :type Fort%20Pawn%20Object%20Reference))
  ;[1 1 1] OnInteracted - Send Controller who Interacted with the Chest to Give Score
  (bind _output_get (Variables|Default|SetInteracting%20Player InteractingPawn))
  (bind _asfort_player_pawn_athena (Utilities|Casting|CastToFortPlayerPawnAthena _output_get))
  (Ability|SendGameplayEventtoActor self "(TagName=\"Athena.LTM.Mash.ScoreMultiplier.Opened\")" (Utilities|Struct|MakeGameplayEventData :Instigator self :OptionalObject (Pawn|GetController _asfort_player_pawn_athena) :ContextHandle "()" :InstigatorTags "(GameplayTags=)" :TargetTags "(GameplayTags=)" :EventMagnitude 0.0 :TargetData "()")))

(event Custom|Mobile_OnDisconnect
  ;[1 1 1] Mobile Wiggle on Interact
  (Networking|CanExecuteCosmeticEvents
    (:True
      (Timeline|MobileOnInteractTL Reverse))))

(event EventOnSetSearched
  (CallFunction|SetLightVisibility false)
  (Audio|Components|Audio|FadeOut (Variables|Default|GetChest_Ambient_Sound) 0.5 0.0 "Linear" :!class /Script/Engine.AudioComponent)
  (Utilities|Time|ClearandInvalidateTimerbyHandle (Variables|Default|GetChimeTimer))
  (Variables|Building|SetbDestroyOnPlayerBuildingPlacement true) ; Allows players to build over the opened chest mesh (but not when there is loot in there)
  (CallFunction|CleanupWiggleMIDs)
  (Rendering|SetRenderCustomDepth (Variables|Debug|GetStaticMeshComponent) false)
  (Development|SetHiddeninGame (Variables|Default|GetUI_PortalInfoPlate) true false)
  (Utilities|Time|ClearandInvalidateTimerbyHandle (Variables|Default|GetUICheckTimer)))

(event EventDestroyed
  (CallFunction|SetLightVisibility false)
  (Audio|Components|Audio|FadeOut (Variables|Default|GetChest_Ambient_Sound) 0.5 0.0 "Linear" :!class /Script/Engine.AudioComponent)
  (Utilities|Time|ClearandInvalidateTimerbyHandle (Variables|Default|GetChimeTimer))
  (Variables|Building|SetbDestroyOnPlayerBuildingPlacement true) ; Allows players to build over the opened chest mesh (but not when there is loot in there)
  (CallFunction|CleanupWiggleMIDs)
  (Rendering|SetRenderCustomDepth (Variables|Debug|GetStaticMeshComponent) false)
  (Development|SetHiddeninGame (Variables|Default|GetUI_PortalInfoPlate) true false)
  (Utilities|Time|ClearandInvalidateTimerbyHandle (Variables|Default|GetUICheckTimer)))

(event EventOnLoot
  (bind _vfxs (Variables|Default|GetVFXs))
  (bind _childactor (Variables|/Script/Engine.ChildActorComponent|GetChildActor _vfxs))
  (bind (_asmash_multiplier_vfx _bsuccess) (Utilities|Casting|CastToMashMultiplierVFX _childactor))
  (Utilities|FlowControl|Sequence
    (:then_0
      (CallFunction|SetLightVisibility false)
      (Audio|Components|Audio|FadeOut (Variables|Default|GetChest_Ambient_Sound) 0.5 0.0 "Linear" :!class /Script/Engine.AudioComponent)
      (Utilities|Time|ClearandInvalidateTimerbyHandle (Variables|Default|GetChimeTimer))
      (Variables|Building|SetbDestroyOnPlayerBuildingPlacement true) ; Allows players to build over the opened chest mesh (but not when there is loot in there)
      (CallFunction|CleanupWiggleMIDs)
      (Rendering|SetRenderCustomDepth (Variables|Debug|GetStaticMeshComponent) false)
      (Development|SetHiddeninGame (Variables|Default|GetUI_PortalInfoPlate) true false)
      (Utilities|Time|ClearandInvalidateTimerbyHandle (Variables|Default|GetUICheckTimer)))
    (:then_1
      (if _bsuccess
        (Class|MashMultiplierVFX|ActivateExplosionVFX _asmash_multiplier_vfx (Variables|Default|GetInteracting%20Player))
        (else
          (Development|PrintString "HELLO I DIDNT WORK")))
      (Audio|PlaySoundatLocation "/Game/Athena/Sounds/LTMs/MASH/MASH_ScoreMultiplier_Collect_Cue.MASH_ScoreMultiplier_Collect_Cue" (Transformation|GetActorLocation))
      (Rendering|ToggleVisibility (Variables|Debug|GetStaticMeshComponent) false)
      (Utilities|FlowControl|Delay 1.0)
      (Actor|DestroyActor))))

(event EventOnDeathPlayEffects
  (params (Damage :type Float%20%28single-precision%29) (DamageTags :type Gameplay%20Tag%20Container%20Structure%20%28by%20ref%29) (Momentum :type Vector :default "0, 0, 0") (HitInfo :type Hit%20Result%20Structure%20%28by%20ref%29) (InstigatedBy :type Fort%20Pawn%20Object%20Reference) (DamageCauser :type Actor%20Object%20Reference) (EffectContext :type Gameplay%20Effect%20Context%20Handle%20Structure))
  (Audio|Components|Audio|FadeOut (Variables|Default|GetChest_Ambient_Sound) 0.5 0.0 "Linear" :!class /Script/Engine.AudioComponent))

(event Custom|CheckShouldDisplayUI
  (bind _ui_portalinfoplate (Variables|Default|GetUI_PortalInfoPlate))
  (bind _returnvalue (UserInterface|GetUserWidgetObject _ui_portalinfoplate))
  (Utilities|IsValid _returnvalue
    (:Is%20Valid
      (bind _asfort_player_controller (Utilities|Casting|CastToFortPlayerController (Widget|GetOwningPlayer _returnvalue)))
      (Utilities|IsValid (Game|Player|GetPlayerPawn _asfort_player_controller)
        (:Is%20Valid
          (bind _distance (Class|MissionBlueprintFunctionLibrary|DistanceBetweenTwoVectors (Transformation|GetWorldLocation (Variables|/Script/Engine.Character|GetMesh (Game|Player|GetPlayerPawn _asfort_player_controller))) (Transformation|GetWorldLocation (Variables|Default|GetUI_PortalInfoPlate))))
          (CallFunction|SetUIVisibility _distance))))))

(event OnProjectileStop(FortProjectileMovement)
  (params (ImpactResult :type Hit%20Result%20Structure%20%28by%20ref%29))
  (bind (_bblockinghit _binitialoverlap _time _distance _location _impactpoint _normal _impactnormal _physmat _hitactor) (Collision|BreakHitResult ImpactResult))
  ;[0 0.9225 1] Projectile Movement
  ; We need to track if this object should be at rest on the client AND the server.
  ; This is necessary because the resting state can toggle back and forth very quickly, faster than replication (example: if the actor falls a small distance like 50uu.)
  ;;[1 0.5091 0] OnProjectile Stopped. If we hit a building come to rest on it (so we can turn off ticking)
  (bind _asbuilding_actor (Utilities|Casting|CastToBuildingActor _hitactor)
    (:then
      (Variables|Default|SetLandedBuilding _asbuilding_actor)
      (EventDispatchers|BindEventtoOnDied _asbuilding_actor (EventDispatchers|CreateEvent :!event Custom|OnLandedBuildingDestroyed)))
    (:CastFailed
      (Utilities|Casting|CastToLandscape _hitactor) ; If this actor stopped on landscape put this actor at rest permenantly.
    )) ; If this actor stopped on a building, listen for the building's destruction
  (CallFunction|SetIsResting true))

(event Custom|OnLandedBuildingDestroyed
  (params (DamagedActor :type Actor%20Object%20Reference) (Damage :type Float%20%28single-precision%29) (InstigatedBy :type Controller%20Object%20Reference) (DamageCauser :type Actor%20Object%20Reference) (HitLocation :type Vector) (FHitComponent :type Primitive%20Component%20Object%20Reference) (BoneName :type Name) (Momentum :type Vector))
  ;[0 0.9225 1] Projectile Movement
  ; We need to track if this object should be at rest on the client AND the server.
  ; This is necessary because the resting state can toggle back and forth very quickly, faster than replication (example: if the actor falls a small distance like 50uu.)
  ;;[0.08305 1 0] OnLandedBuildingDestroyed, when the building we are resting on is destroyed set IsResting to false. This will restart projectile movement (if it was disabled.)
  (CallFunction|SetIsResting false))

(event Custom|Set%20Is%20Resting
  (params (In%20Is%20Resting :type Boolean))
  ;[0.1033 0.0772 1] SetIsResting
  (if (!= (Variables|Default|GetbIsAtRestLocal) In%20Is%20Resting)
    (Variables|Default|SetbIsAtRestLocal In%20Is%20Resting)
    (Networking|SwitchHasAuthority
      (:Authority
        ;;[0.06105 1 0.2183] Replicate to the client the new At Rest State of the server.
        (Variables|Default|SetbIsAtRest In%20Is%20Resting))
      (:Remote
        (CallFunction|UpdateProjectileMovement)))))

(event Custom|UpdateProjectileMovement
  (bind _fortprojectilemovement (Variables|Default|GetFortProjectileMovement))
  ;[0 0.9007 1] UpdateProjectileMovement
  ; Enable/Disable projectile movement based on if this actor is at rest.
  ; We enable projectile movement if either the replicated and local variables for tracking is at rest are false.
  ; Only disable projectile movement if both variables agree that this object should be at rest.
  (if (and (Variables|Default|GetbIsAtRest) (Variables|Default|GetbIsAtRestLocal))
    ;;[1 0.4859 0] We are resting on a building, disable projectile movement and ticking.
    (Components|Movement|SetUpdatedComponent _fortprojectilemovement 0)
    (Variables|/Script/Engine.MovementComponent|SetVelocity "0, 0, 0" _fortprojectilemovement)
    (CallFunction|SetTemporarilyTicking false)
    (else
      ;;[0.08708 1 0] We are no longer resting on a building, restart projectile movement and ticking.
      (Components|Movement|SetUpdatedComponent _fortprojectilemovement (Variables|Debug|GetStaticMeshComponent))
      (CallFunction|SetTemporarilyTicking true))))

(fn BlueprintCanInteract
  (params (InteractingPawn :type Fort%20Pawn%20Object%20Reference) (InteractionBeingAttempted :type EInteractionBeingAttempted%20Enum :default "FirstInteraction") (InteractionType :type TInteractionType%20Enum :default "IT_NoInteraction"))
  (bind _asfort_player_controller_outpost (Utilities|Casting|CastToFortPlayerControllerOutpost (Pawn|GetController InteractingPawn))
    (:then
      (return (Outpost|IsOutpostOwner _asfort_player_controller_outpost)))
    (:CastFailed
      (return true))))

(fn CleanupWiggleMIDs
  (if (TouchPlatform|IsMobileGame)
    (Components|Timeline|Stop (Variables|ScoreMultiplier_Mash|GetMobileOnInteractTL))
    (for _array_element (Variables|Default|GetMIDs)
      (Rendering|Material|SetScalarParameterValue _array_element "WiggleOnInteractAmount" 0.0 :!class /Script/Engine.MaterialInstanceDynamic))))

(fn CreateMobileMIDs
  (if (TouchPlatform|IsMobileGame)
    (for (_ _array_index) (Rendering|Material|GetMaterials (Variables|Debug|GetStaticMeshComponent))
      (bind _returnvalue (CallFunction|GetorCreateMIDOnDefaultMesh _array_index))
      (Utilities|Array|SetArrayElem (Variables|Default|GetMIDs) _array_index _returnvalue true))))

(fn GetCurrentLocation
  (return "0, 0, 0"))

(fn GetFortProjectileMovementComponent
  (return 0))

(fn GetInstigatorPlayerController
  (return 0))

(fn GetMaxAudibleDistance
  (bind _chest_ambient_sound (Variables|Default|GetChest_Ambient_Sound))
  (bind _sound (Variables|/Script/Engine.AudioComponent|GetSound _chest_ambient_sound))
  (Utilities|IsValid _sound
    (:Is%20Valid
      (return (Variables|/Script/Engine.SoundBase|GetMaxDistance _sound)))
    (:Is%20Not%20Valid
      (return 0.0))))

(fn GetMaxTravelDistanceSquared
  (return 0.0))

(fn IsProjectileBeingKilled
  (return false))

(fn OnRep_bIsAtRest
  (CallFunction|UpdateProjectileMovement))

(fn OnSetCustomDepthStencilValue
  (params (PrimComponents :type Array%20of%20Primitive%20Component%20Object%20References) (bUseCustomDepth :type Boolean) (StencilValue :type Integer))
  (for _array_element PrimComponents
    (Rendering|SetCustomDepthStencilValue _array_element (Variables|GetStencilValue))
    (if (Variables|GetbUseCustomDepth)
      (Rendering|SetRenderCustomDepth _array_element true)))
  (return true))

(fn SetLightVisibility
  (params (Visible :type Boolean))
  (Variables|Default|GetTrasureLight
    (:then
      (Rendering|SetVisibility (Variables|Default|GetTrasureLight) Visible false)
      (if Visible
        (Audio|PlaySoundatLocation 0 (Transformation|GetActorLocation))))
    (:else)))

(fn SetUIVisibility
  (params (Distance :type Float%20%28double-precision%29))
  (bind _ui_portalinfoplate (Variables|Default|GetUI_PortalInfoPlate))
  (bind _ashud_ss_mash_multiplier (Utilities|Casting|CastToHUD_SS_MashMultiplier (UserInterface|GetUserWidgetObject _ui_portalinfoplate)))
  (Class|HUDSSMashMultiplier|EventShowDetail _ashud_ss_mash_multiplier (select (< Distance (Variables|Default|GetUIDetail%20Distance)) true false))
  (Development|SetHiddeninGame _ui_portalinfoplate (select (< Distance (Variables|Default|GetUIShow%20Distance)) false true) false))

(fn ShouldClampTravelDistance
  (return false))

(fn ShouldDie
  (params (Damage :type Float%20%28single-precision%29) (EventInstigator :type Controller%20Object%20Reference) (DamageCauser :type Actor%20Object%20Reference) (bIsTest :type Boolean) (EffectContext :type Gameplay%20Effect%20Context%20Handle%20Structure%20%28by%20ref%29))
  (bind _asfort_player_controller_outpost (Utilities|Casting|CastToFortPlayerControllerOutpost EventInstigator)
    (:then
      (return (Outpost|IsOutpostOwner _asfort_player_controller_outpost)))
    (:CastFailed
      (return true))))

(fn ConstructionScript
  (Components|Movement|SetUpdatedComponent (Variables|Default|GetFortProjectileMovement) (Variables|Debug|GetStaticMeshComponent))
  (Networking|CanExecuteCosmeticEvents
    (:True
      (Utilities|EffectsQualityCheck 0
        (:Medium
          _)
        (:Low
          _))
      (Components|DestroyComponent (Variables|Default|GetTrasureLight)))))
